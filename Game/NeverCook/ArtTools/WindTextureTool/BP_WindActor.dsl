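; Wind actor that turns its rotation and Speed into global wind data for grass and medium foliage.
; Float curves map Speed to gust-scale parameters on two dynamic materials, drawn into separate render targets, while MPC_Wind receives direction, angle and strength.
; Updates run on construction, BeginPlay, ManualUpdate and a one-second looping timer, which recreates the medium material each time.
; The medium update writes its Direction parameter to the grass material.
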(event EventBeginPlay
  (Utilities|Time|SetTimerbyFunctionName 0 "TimerUpdate" 1.0 true)
  (bind _returnvalue (Rendering|Material|CreateDynamicMaterialInstance "/Game/NeverCook/ArtTools/WindTextureTool/M_WindRTMAT.M_WindRTMAT" "None" "None" :!class /Script/Engine.KismetMaterialLibrary))
  (Variables|Default|SetMID _returnvalue)
  (CallFunction|UpdateWindTexture (Variables|Default|GetSpeed) (Math|Vector|GetForwardVector (Transformation|GetActorRotation)))
  (bind _returnvalue_2 (Rendering|Material|CreateDynamicMaterialInstance "/Game/NeverCook/ArtTools/WindTextureTool/MediumFoliage/M_WindRTMAT_MediumFoliage.M_WindRTMAT_MediumFoliage" "None" "None" :!class /Script/Engine.KismetMaterialLibrary))
  (Variables|Default|SetMID_Medium _returnvalue_2)
  (CallFunction|UpdateWindTextureMedium (Variables|Default|GetSpeed) (Math|Vector|GetForwardVector (Transformation|GetActorRotation))))

(event Collision|EventActorBeginOverlap
  (params (OtherActor :type Actor%20Object%20Reference)))

(event EventTick
  (params (DeltaSeconds :type Float%20%28single-precision%29)))

(event Custom|TimerUpdate
  (CallFunction|UpdateWindTexture (Variables|Default|GetSpeed) (Math|Vector|GetForwardVector (Transformation|GetActorRotation)))
  (bind _returnvalue (Rendering|Material|CreateDynamicMaterialInstance "/Game/NeverCook/ArtTools/WindTextureTool/MediumFoliage/M_WindRTMAT_MediumFoliage.M_WindRTMAT_MediumFoliage" "None" "None" :!class /Script/Engine.KismetMaterialLibrary))
  (Variables|Default|SetMID_Medium _returnvalue)
  (CallFunction|UpdateWindTextureMedium (Variables|Default|GetSpeed) (Math|Vector|GetForwardVector (Transformation|GetActorRotation))))

(event Custom|ManualUpdate
  (bind _returnvalue (Rendering|Material|CreateDynamicMaterialInstance "/Game/NeverCook/ArtTools/WindTextureTool/M_WindRTMAT.M_WindRTMAT" "None" "None" :!class /Script/Engine.KismetMaterialLibrary))
  (Variables|Default|SetMID _returnvalue)
  (CallFunction|UpdateWindTexture (Variables|Default|GetSpeed) (Math|Vector|GetForwardVector (Transformation|GetActorRotation)))
  (bind _returnvalue_2 (Rendering|Material|CreateDynamicMaterialInstance "/Game/NeverCook/ArtTools/WindTextureTool/MediumFoliage/M_WindRTMAT_MediumFoliage.M_WindRTMAT_MediumFoliage" "None" "None" :!class /Script/Engine.KismetMaterialLibrary))
  (Variables|Default|SetMID_Medium _returnvalue_2)
  (CallFunction|UpdateWindTextureMedium (Variables|Default|GetSpeed) (Math|Vector|GetForwardVector (Transformation|GetActorRotation))))

(fn UpdateWindTexture
  (params (Speed :type Float%20%28double-precision%29) (Direction :type Vector))
  (bind _direction (Variables|GetDirection))
  (bind _returnvalue (* _direction "1.000000,1.000000,0"))
  (bind _returnvalue_2 (Math|Vector|Normalize _returnvalue))
  (bind _mid (Variables|Default|GetMID))
  (Rendering|Material|SetScalarParameterValue _mid "Huge" (Math|Curves|GetFloatValue (Variables|Default|GetHuge) Speed) :!class /Script/Engine.MaterialInstanceDynamic)
  (Rendering|Material|SetScalarParameterValue _mid "Big" (Math|Curves|GetFloatValue (Variables|Default|GetBig) Speed) :!class /Script/Engine.MaterialInstanceDynamic)
  (Rendering|Material|SetScalarParameterValue _mid "Medium" (Math|Curves|GetFloatValue (Variables|Default|GetMedium) Speed) :!class /Script/Engine.MaterialInstanceDynamic)
  (Rendering|Material|SetScalarParameterValue _mid "Small" (Math|Curves|GetFloatValue (Variables|Default|GetSmall) Speed) :!class /Script/Engine.MaterialInstanceDynamic)
  (Rendering|DrawMaterialtoRenderTarget "/Game/Environments/Asteria/Foliage/Wind/RT_WindGlobal_Grass.RT_WindGlobal_Grass" (Variables|Default|GetMID))
  (Rendering|Material|SetVectorParameterValue "/Game/Environments/Asteria/Foliage/Wind/MPC_Wind.MPC_Wind" "WindDirection" (Math|Conversions|ToLinearColor(Vector) _returnvalue_2) :!class /Script/Engine.KismetMaterialLibrary)
  (Rendering|Material|SetScalarParameterValue "/Game/Environments/Asteria/Foliage/Wind/MPC_Wind.MPC_Wind" "WindAngle" (Class|AbilityFunctions|GetAngleBetweenTwoVectors _returnvalue_2 "1.000000, 0, 0") :!class /Script/Engine.KismetMaterialLibrary)
  (Rendering|Material|SetScalarParameterValue "/Game/Environments/Asteria/Foliage/Wind/MPC_Wind.MPC_Wind" "WindStrength" (Variables|Default|GetSpeed) :!class /Script/Engine.KismetMaterialLibrary))

(fn UpdateWindTexture_Medium
  (params (Speed :type Float%20%28double-precision%29) (Direction :type Vector))
  (bind _direction (Variables|GetDirection))
  (bind _returnvalue (* _direction "1.000000,1.000000,0"))
  (bind _returnvalue_2 (Math|Vector|Normalize _returnvalue))
  (bind _mid_medium (Variables|Default|GetMID_Medium))
  (bind _mid_medium_2 (Variables|Default|GetMID_Medium))
  (Rendering|Material|SetScalarParameterValue _mid_medium "Huge" (Math|Curves|GetFloatValue (Variables|Default|GetHuge_Medium) Speed) :!class /Script/Engine.MaterialInstanceDynamic)
  (Rendering|Material|SetScalarParameterValue _mid_medium "Big" (Math|Curves|GetFloatValue (Variables|Default|GetBig_Medium) Speed) :!class /Script/Engine.MaterialInstanceDynamic)
  (Rendering|Material|SetScalarParameterValue _mid_medium "Medium" (Math|Curves|GetFloatValue (Variables|Default|GetMedium_Medium) Speed) :!class /Script/Engine.MaterialInstanceDynamic)
  (Rendering|Material|SetScalarParameterValue _mid_medium "Small" (Math|Curves|GetFloatValue (Variables|Default|GetSmall_Medium) Speed) :!class /Script/Engine.MaterialInstanceDynamic)
  (Rendering|Material|SetScalarParameterValue _mid_medium_2 "Huge2" (Math|Curves|GetFloatValue (Variables|Default|GetStyle2_Huge_Medium) Speed) :!class /Script/Engine.MaterialInstanceDynamic)
  (Rendering|Material|SetScalarParameterValue _mid_medium_2 "Big2" (Math|Curves|GetFloatValue (Variables|Default|GetStyle2_Big_Medium) Speed) :!class /Script/Engine.MaterialInstanceDynamic)
  (Rendering|Material|SetScalarParameterValue _mid_medium_2 "Medium2" (Math|Curves|GetFloatValue (Variables|Default|GetStyle2_Medium_Medium) Speed) :!class /Script/Engine.MaterialInstanceDynamic)
  (Rendering|Material|SetScalarParameterValue _mid_medium_2 "Small2" (Math|Curves|GetFloatValue (Variables|Default|GetStyle2_Small_Medium) Speed) :!class /Script/Engine.MaterialInstanceDynamic)
  (Rendering|Material|SetVectorParameterValue (Variables|Default|GetMID) "Direction" (Math|Conversions|ToLinearColor(Vector) (+ (/ _returnvalue_2 2.0) 1.0)) :!class /Script/Engine.MaterialInstanceDynamic)
  (Rendering|DrawMaterialtoRenderTarget "/Game/Environments/Asteria/Foliage/Wind/RT_WindGlobal_Medium.RT_WindGlobal_Medium" (Variables|Default|GetMID_Medium))
  (Rendering|Material|SetVectorParameterValue "/Game/Environments/Asteria/Foliage/Wind/MPC_Wind.MPC_Wind" "WindDirection" (Math|Conversions|ToLinearColor(Vector) _returnvalue_2) :!class /Script/Engine.KismetMaterialLibrary)
  (Rendering|Material|SetScalarParameterValue "/Game/Environments/Asteria/Foliage/Wind/MPC_Wind.MPC_Wind" "WindAngle" (Class|AbilityFunctions|GetAngleBetweenTwoVectors _returnvalue_2 "1.000000, 0, 0") :!class /Script/Engine.KismetMaterialLibrary)
  (Rendering|Material|SetScalarParameterValue "/Game/Environments/Asteria/Foliage/Wind/MPC_Wind.MPC_Wind" "WindStrength" (Variables|Default|GetSpeed) :!class /Script/Engine.KismetMaterialLibrary))

(fn ConstructionScript
  (bind _returnvalue (Transformation|GetActorRotation))
  (bind _returnvalue_2 (Math|Vector|GetForwardVector _returnvalue))
  (bind _speed (Variables|Default|GetSpeed))
  (Utilities|FlowControl|Sequence
    (:then_0
      (bind _returnvalue_3 (Rendering|Material|CreateDynamicMaterialInstance "/Game/NeverCook/ArtTools/WindTextureTool/M_WindRTMAT.M_WindRTMAT" "None" "None" :!class /Script/Engine.KismetMaterialLibrary))
      (Variables|Default|SetMID _returnvalue_3)
      (CallFunction|UpdateWindTexture _speed _returnvalue_2))
    (:then_1
      (bind _returnvalue_4 (Rendering|Material|CreateDynamicMaterialInstance "/Game/NeverCook/ArtTools/WindTextureTool/MediumFoliage/M_WindRTMAT_MediumFoliage.M_WindRTMAT_MediumFoliage" "None" "None" :!class /Script/Engine.KismetMaterialLibrary))
      (Variables|Default|SetMID_Medium _returnvalue_4)
      (CallFunction|UpdateWindTextureMedium _speed _returnvalue_2))))
