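# Unlit material in the wind texture tool, apparently for rendering a wind texture that combines direction and strength noise.
# Huge, Big, Medium and Small parameters weight noise masks, overlay-blended into strength and, through NormalFromHeightmap and angle-corrected blending, into a direction normal.
# One branch multiplies that flattened, renormalized normal by strength.
# Another skews the MPC_Wind direction sideways by the noise, packing XY with strength; no material output connection appears for either.
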Material /Game/NeverCook/ArtTools/WindTextureTool/M_WindRTMAT
graph {
 properties {pixelDepthOffsetMode=PDOM_Legacy, shadingModel=MSM_Unlit, usedShadingModels=Unlit}
 nodes [
  n1 = Constant3Vector() {constant=<rgba 1 1 0 0>}
  n2 = ComponentMask(None=CollectionParameter() {collection=/Game/Environments/Asteria/Foliage/Wind/MPC_Wind, parameterName=WindDirection}) {b=true, g=true, r=true}
  n3 = Constant3Vector() {constant=<rgba 0 0 1 0>}
  n4 = Multiply(
   A=ScalarParameter() {defaultValue=1, parameterName=Huge, sliderMax=1}
   B=TextureSample() {samplerType=SAMPLERTYPE_LinearColor, texture=&/WindNoise_Mask_5}.R
  )
  n5 = LinearInterpolate(
   A=@n4
   B=MaterialFunctionCall(Base=@n4, Blend=TextureSample() {samplerType=SAMPLERTYPE_LinearColor, texture=&/WindNoise_Mask_2}.R) {materialFunction=&Blend_Overlay}.Result
   Alpha=ScalarParameter() {defaultValue=0.75, parameterName=Big}
  )
  n6 = LinearInterpolate(
   A=@n5
   B=MaterialFunctionCall(Base=@n5, Blend=TextureSample() {samplerType=SAMPLERTYPE_LinearColor, texture=&/WindNoise_Mask_3}.R) {materialFunction=&Blend_Overlay}.Result
   Alpha=ScalarParameter() {defaultValue=0.45, parameterName=Medium}
  )
  n7 = LinearInterpolate(
   A=@n6
   B=MaterialFunctionCall(Base=@n6, Blend=TextureSample() {samplerType=SAMPLERTYPE_LinearColor, texture=&/WindNoise_Mask_4}.R) {materialFunction=&Blend_Overlay}.Result
   Alpha=ScalarParameter() {defaultValue=0.2, parameterName=Small}
  )
  n8 = Subtract(A=ComponentMask(None=@n7) {r=true}) {constB=0.5}
  n9 = ScalarParameter() {defaultValue=2, parameterName=WindNoise}
  n10 = Add(A=@n2, B=Multiply(A=CrossProduct(A=@n2, B=@n3), B=Multiply(A=@n8, B=@n9)))
  n11 = Add(A=Normalize(VectorInput=Multiply(A=@n1, B=@n10)))
  AppendVector(A=ComponentMask(None=Divide(A=@n11)) {g=true, r=true}, B=ComponentMask(None=@n7) {r=true})
  n12 = ScalarParameter() {defaultValue=2.7977, parameterName=WindDirectionNoise}
  n13 = ScalarParameter() {defaultValue=0.005}
  n14 = LinearInterpolate(
   A=Constant3Vector() {constant=<rgba 0 0 1 0>}
   B=MaterialFunctionCall!(
    Height Map=TextureObject() {samplerType=SAMPLERTYPE_Masks, texture=/Game/Global/Textures/Default/Noise/WindNoise_Mask_1}
    Normal Map Intensity=@n12
    Height Map UV Offset=@n13
    Coordinates
    Height Map Channel Selector
   ) {materialFunction=/Engine/Functions/Engine_MaterialFunctions03/Procedurals/NormalFromHeightmap}.Result
   Alpha=ScalarParameter() {defaultValue=1, parameterName=Huge, sliderMax=1}
  )
  n15 = TextureObject() {samplerType=SAMPLERTYPE_Masks, texture=&WindNoise_Mask_2}
  n16 = LinearInterpolate(
   A=@n14
   B=MaterialFunctionCall(
    BaseNormal=@n14
    AdditionalNormal=MaterialFunctionCall(Height Map=@n15, Normal Map Intensity=@n12, Height Map UV Offset=@n13) {materialFunction=&NormalFromHeightmap}.Result
   ) {materialFunction=&BlendAngleCorrectedNormals}.Result
   Alpha=ScalarParameter() {defaultValue=0.75, parameterName=Big}
  )
  n17 = TextureObject() {samplerType=SAMPLERTYPE_Masks, texture=&WindNoise_Mask_3}
  n18 = LinearInterpolate(
   A=@n16
   B=MaterialFunctionCall(
    BaseNormal=@n16
    AdditionalNormal=MaterialFunctionCall(Height Map=@n17, Normal Map Intensity=@n12, Height Map UV Offset=@n13) {materialFunction=&NormalFromHeightmap}.Result
   ) {materialFunction=&BlendAngleCorrectedNormals}.Result
   Alpha=ScalarParameter() {defaultValue=0.45, parameterName=Medium}
  )
  n19 = MaterialFunctionCall(
   BaseNormal=@n18
   AdditionalNormal=MaterialFunctionCall(
    Height Map=TextureObject() {samplerType=SAMPLERTYPE_Masks, texture=&WindNoise_Mask_4}
    Normal Map Intensity=@n12
    Height Map UV Offset=@n13
   ) {materialFunction=&NormalFromHeightmap}.Result
  ) {materialFunction=&BlendAngleCorrectedNormals}
  n20 = ScalarParameter() {defaultValue=0.2, parameterName=Small}
  n21 = ComponentMask(None=Normalize(VectorInput=LinearInterpolate(A=@n18, B=@n19.Result, Alpha=@n20))) {g=true, r=true}
  n22 = Constant() {r=1}
  Multiply(A=Normalize(VectorInput=AppendVector(A=@n21, B=@n22)), B=@n7)
 ]
}
